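; Dual-wield Fashion Feline pickaxe Blueprint that adds cosmetic swing VFX on top of its parent class.
; Each of the four left and right swing events calls the parent, then sets SwingSectionIndex on the swing VFX component from the instigator's montage section, or -1 if the cast fails.
; The second left swing also spawns a Niagara trail attached to the weapon mesh and sets IsLastSwing on it.
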(event EventSwingRight
  (bind _returnvalue (CallFunction|GetSwingVFXComponent))
  ;[1 1 1] Final Swing
  (|CallParentFunction)
  (bind _section_index (Utilities|SetupSwingMontageSectionIndex (Game|GetInstigator))
    (:Out
      (Effects|Components|ParticleSystem|SetIntParameter _returnvalue "SwingSectionIndex" _section_index))
    (:Cast%20Failed%20Out
      (Effects|Components|ParticleSystem|SetIntParameter _returnvalue "SwingSectionIndex" -1))))

(event EventSwingLeft2
  (bind _returnvalue (CallFunction|GetSwingVFXComponent))
  ;[1 1 1] Final Swing
  (|CallParentFunction)
  (bind _returnvalue_2 (Niagara|SpawnSystemAttached "/BRCosmetics/Weapons/FORT_Melee/Pickaxe_FashionFeline/FX/NS_Pickaxe_FashionFeline_Trail.NS_Pickaxe_FashionFeline_Trail" (Variables|Visual|GetWeaponMesh) "None" "0, 0, 0" "0, 0, 0" "KeepRelativeOffset" true))
  (Effects|Components|ParticleSystem|SetBoolParameter _returnvalue_2 "IsLastSwing" true)
  (bind _section_index (Utilities|SetupSwingMontageSectionIndex (Game|GetInstigator))
    (:Out
      (Effects|Components|ParticleSystem|SetIntParameter _returnvalue "SwingSectionIndex" _section_index))
    (:Cast%20Failed%20Out
      (Effects|Components|ParticleSystem|SetIntParameter _returnvalue "SwingSectionIndex" -1))))

(event EventSwingLeft
  (bind _returnvalue (CallFunction|GetSwingVFXComponent))
  ;[1 1 1] Final Swing
  (|CallParentFunction)
  (bind _section_index (Utilities|SetupSwingMontageSectionIndex (Game|GetInstigator))
    (:Out
      (Effects|Components|ParticleSystem|SetIntParameter _returnvalue "SwingSectionIndex" _section_index))
    (:Cast%20Failed%20Out
      (Effects|Components|ParticleSystem|SetIntParameter _returnvalue "SwingSectionIndex" -1))))

(event EventSwingRight2
  (bind _returnvalue (CallFunction|GetSwingVFXComponent))
  ;[1 1 1] Final Swing
  (|CallParentFunction)
  (bind _section_index (Utilities|SetupSwingMontageSectionIndex (Game|GetInstigator))
    (:Out
      (Effects|Components|ParticleSystem|SetIntParameter _returnvalue "SwingSectionIndex" _section_index))
    (:Cast%20Failed%20Out
      (Effects|Components|ParticleSystem|SetIntParameter _returnvalue "SwingSectionIndex" -1))))

(fn ConstructionScript
  (|CallParentFunction))
