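# Niagara ribbon swing trail for the FashionFeline pickaxe, built from a single short-lived Trail emitter.
# Anim-trail modules spawn and update the ribbon between two bone or socket indices, with User.SwingSectionIndex selecting the pair and signed width.
# IsLastSwing feeds the first width case, but both its branches are identical.
# Color scales by a time-of-day multiplier, with alpha also shaped by a rise-and-fall curve and system fade-out.
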NiagaraSystem /BRCosmetics/Weapons/FORT_Melee/Pickaxe_FashionFeline/FX/NS_Pickaxe_FashionFeline_Trail
dependencies {
 usedDataInterfaces [&NiagaraDataInterfaceCurve, &NiagaraDataInterfaceSocketReader, &NiagaraDataInterfaceColorCurve]
 usedDynamicInputs [&ColorFromCurve]
 usedModules [
  &self:SystemSpawnScript.SystemScriptSource.SystemScriptGraph.NiagaraNodeAssignment_1.SetVariables_#1
  &SystemState
  &self:SystemSpawnScript.SystemScriptSource.SystemScriptGraph.NiagaraNodeAssignment_0.SetVariables_#2
  &self:SystemSpawnScript.SystemScriptSource.SystemScriptGraph.NiagaraNodeAssignment_2.SetVariables_#3
 ]
}
emitters {Trail {
  emitterUpdateScript [
   &EmitterState(Life Cycle Mode=::NewEnumerator1, Loop Behavior=::NewEnumerator1, Loop Duration=0.2, UseLoopDelay=false)
   &FN_SpawnAnimTrail(Spawn Interval=0)
  ]
  particleSpawnScript [
   &FN_InitializeAnimTrailBoneIndecies(
    First Bone or Socket Index:NiagaraInt32=&MergeHelper_IntSwitch(0=0, 1=0, 2=4, 3=0, 4=?, 5=?, Index=$User.SwingSectionIndex, Num=::NewEnumerator2)
    Second Bone or Socket Index:NiagaraInt32=&MergeHelper_IntSwitch(0=1, 1=1, 2=5, 3=1, 4=?, 5=?, Index=$User.SwingSectionIndex, Num=::NewEnumerator2)
   )
   &FN_InitializeAnimTrail(
    AnimTrailWidth=&Multiply_Float(
     A=1.2
     B=&MergeHelper_FloatSwitch(
      0=&MakeCustomFloatFromBool(Bool=$User.IsLastSwing, True Float=1.5, False Float=1.5)
      1=-1.5
      2=1.5
      3=-1.5
      Index=$User.SwingSectionIndex
      Num=::NewEnumerator2
     )
    )
    Lifetime=0.2
   )
  ]
  particleUpdateScript [
   &ParticleState()
   &FN_UpdateAnimTrail() {module=FN_UpdateAnimTrail001}
   &ScaleColor(
    Scale Alpha=&Multiply_Float(
     A=&Multiply_Float(
      A=&MakeFloatFromLinearColor(LinearColor=$System.TODColorMultiplier, Channel:ENiagaraFloatFromLinearColorOptions=::NewEnumerator3)
      B=&FloatFromCurve(FloatCurve=DataInterface() {propertyValues {Curve {
          defaultValue=1
          keys [
           {arriveTangent=9.0725, interpMode=RCIM_Cubic, leaveTangent=9.0725, tangentMode=RCTM_User, time=0, value=0}
           {arriveTangent=-2.5249, interpMode=RCIM_Cubic, leaveTangent=-2.5249, tangentMode=RCTM_User, time=0.1408, value=0.9847}
           {arriveTangent=-0.0405, interpMode=RCIM_Cubic, leaveTangent=-0.0405, tangentMode=RCTM_User, time=1, value=0}
          ]
         }}})
     )
     B=$System.SystemFadeOut
    )
    Scale RGB=&Multiply_Vector(
     A=&VectorFromFloat(Value=1.2)
     B=&MakeVectorFromLinearColorRGB(Color=$System.TODColorMultiplier)
    )
    ScaleA=true
    ScaleRGB=true
   )
  ]
  rendererClasses [&NiagaraRibbonRendererProperties]
  renderers [{rendererClass=&NiagaraRibbonRendererProperties}]
 }}
summary {
 systemName=NS_Pickaxe_FashionFeline_Trail
 emitters [{emitterName=Trail, rendererClasses [&NiagaraRibbonRendererProperties]}]
 userVariables [
  User.bIsFrontEnd {type=NiagaraBool}
  User.bIsFrontEndPreview {type=NiagaraBool}
  User.IsLastSwing {type=NiagaraBool}
  User.PhysicalSurface {type=NiagaraInt32}
  User.RibbonIDOffset {type=NiagaraInt32}
  User.SwingSectionIndex {type=NiagaraInt32}
 ]
}
